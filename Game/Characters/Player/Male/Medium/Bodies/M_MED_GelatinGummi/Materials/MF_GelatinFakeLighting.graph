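# Material function for the GelatinGummi character body that fakes sunlight shading and a highlight from the global sunlight direction parameter.
# It blends a half-scaled input normal onto the pixel normal in tangent space, then converts the result to world space.
# The main output is a wrapped, half-Lambert style diffuse term from that normal.
# The highlight output smoothsteps the sun direction against the camera reflection vector, scaled down to a faint level.
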MaterialFunction /Game/Characters/Player/Male/Medium/Bodies/M_MED_GelatinGummi/Materials/MF_GelatinFakeLighting
graph {nodes [
  n1 = CollectionParameter() {
   collection=&FortniteMaterialParameters
   parameterName=SunlightDirectionalVector
  }
  n2 = PixelNormalWS()
  n3 = FunctionInput() {inputName=NormalIn}
  n4 = Transform(None=MaterialFunctionCall(
    BaseNormal=Transform(None=@n2) {transformSourceType=TRANSFORMSOURCE_World, transformType=TRANSFORM_Tangent}
    AdditionalNormal=Multiply(A=@n3) {constB=0.5}
   ) {materialFunction=&BlendAngleCorrectedNormals}.Result)
  n5 = Multiply(A=Add(A=DotProduct(A=@n1, B=@n4))) {constB=0.5}
  FunctionOutput(None=Multiply(A=Saturate(None=@n5)))
  n6 = DotProduct(
   A=@n1
   B=MaterialFunctionCall!(Normal=@n4, CameraVector=CameraVectorWS()) {materialFunction=/Engine/Functions/Engine_MaterialFunctions02/Utility/CustomReflectionVector}.Reflection Vector
  )
  FunctionOutput(None=Multiply(A=SmoothStep(Value=@n6)) {constB=0.02}) {outputName=highlight}
 ]}
