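# Material function for the Gelatin Gummi player body, overlaying a translucent-looking, pulsing gel interior on incoming material attributes.
# Three bump-offset samples of an FX mask form shallow, mid and deep layers, tinted by time-animated curve-atlas colors and lit by fake lighting.
# A pre-skinned radial mask drives color pulses and depth wobble, while Fresnel and emissive terms add glow.
# Static switches provide debug views of individual layers.
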MaterialFunction /Game/Characters/Player/Male/Medium/Bodies/M_MED_GelatinGummi/Materials/MF_Gelatin
graph {nodes [
  n1 = Constant()
  n2 = SetMaterialAttributes!(MaterialAttributes=NamedRerouteUsage(), BaseColor, Metallic=@n1, Specular=@n1, Roughness=@n1, EmissiveColor=@n1, Normal, Opacity) {attributeSetTypes [
    #1
    #2
    #3
    #4
    #5
    #6
    #7
   ]}
  n3 = SetMaterialAttributes!(MaterialAttributes=@n2, BaseColor=NamedRerouteUsage()) {attributeSetTypes [#1]}
  n4 = NamedRerouteUsage()
  n5 = SetMaterialAttributes(MaterialAttributes=@n2, BaseColor=NamedRerouteUsage()) {attributeSetTypes [#1]}
  n6 = NamedRerouteUsage()
  n7 = OneMinus(None=NamedRerouteUsage())
  n8 = SetMaterialAttributes(MaterialAttributes=@n2, BaseColor=@n7) {attributeSetTypes [#1]}
  n9 = NamedRerouteUsage()
  n10 = StaticSwitchParameter(
   True=SetMaterialAttributes(MaterialAttributes=@n2, BaseColor=NamedRerouteUsage()) {attributeSetTypes [#1]}
   False=StaticSwitchParameter(
    True=SetMaterialAttributes(MaterialAttributes=@n2, BaseColor=@n6) {attributeSetTypes [#1]}
    False=StaticSwitchParameter(True=@n8, False=@n9) {group=Gelatin - Debug, parameterName=Gelatin Visualize FX Mask}
   ) {group=Gelatin - Debug, parameterName=Gelatin Visualize Deep, sortPriority=102}
  ) {group=Gelatin - Debug, parameterName=Gelatin Visualize Mid, sortPriority=101}
  n11 = StaticSwitchParameter(
   True=SetMaterialAttributes(MaterialAttributes=@n2, BaseColor=NamedRerouteUsage()) {attributeSetTypes [#1]}
   False=StaticSwitchParameter(
    True=SetMaterialAttributes(MaterialAttributes=@n2, BaseColor=@n4) {attributeSetTypes [#1]}
    False=StaticSwitchParameter(True=@n5, False=@n10) {group=Gelatin - Debug, parameterName=Gelatin Visualize Shallow, sortPriority=100}
   ) {group=Gelatin - Debug, parameterName=Gelatin Visualize Shallow Color, sortPriority=0}
  ) {group=Gelatin - Debug, parameterName=Gelatin Visualize Mid Color, sortPriority=0}
  FunctionOutput(None=StaticSwitchParameter(
    True=StaticSwitchParameter(True=@n3, False=@n11) {group=Gelatin - Debug, parameterName=Gelatin Visualize Deep Wobble, sortPriority=0}
    False=NamedRerouteUsage()
   ) {group=Gelatin - Debug, parameterName=Gelatin Debug, sortPriority=0})
  n12 = VertexInterpolator(VS=PreSkinnedPosition())
  n13 = VectorParameter() {group=Gelatin, parameterName=Gelatin PulseCenter, sortPriority=10, defaultValue=<rgba 0 0 150 1>}
  n14 = ScalarParameter() {defaultValue=0.3, group=Gelatin, parameterName=Gelatin Color Pulse FMod, sortPriority=8}
  n15 = Divide(A=Fmod(A=SphereMask(A=@n12.PS, B=@n13.RGB) {attenuationRadius=120, hardnessPercent=0}, B=@n14), B=@n14)
  n16 = Time()
  n17 = Multiply(A=@n16, B=ScalarParameter() {defaultValue=1, group=Gelatin, parameterName=Gelatin Mid ColorPulse Speed, sortPriority=20})
  n18 = CurveAtlasRowParameter(CurveTime=Frac(None=Add(A=@n15, B=@n17))) {
   atlas=&AthenaCurveAtlas01
   curve=/Game/Characters/Player/Male/Medium/Bodies/M_MED_GelatinGummi/Textures/T_GelatinGummi_Mid
   defaultValue=324
   group=Gelatin
   parameterName=Gelatin Mid ColorCurve
   sortPriority=20
  }
  NamedRerouteDeclaration(None=@n18) {name=ColorPulse Mid, nodeColor=<rgba 0 0.9062 1 1>}
  n19 = SphereMask(A=@n12.PS, B=@n13.RGB) {attenuationRadius=120, hardnessPercent=0}
  n20 = ScalarParameter() {defaultValue=0.6, group=Gelatin, parameterName=Gelatin Wobble FMod, sortPriority=8}
  n21 = Add(
   A=Divide(A=Fmod(A=@n19, B=@n20), B=@n20)
   B=Multiply(A=Time(), B=ScalarParameter() {defaultValue=0.2, group=Gelatin, parameterName=Gelatin Wobble Speed, sortPriority=8}) {constB=0.2}
  )
  n22 = Frac(None=Add(A=ScalarParameter() {defaultValue=0.1, group=Gelatin, parameterName=Gelatin Shallow Wobble Offset, sortPriority=10}, B=@n21))
  n23 = Constant() {r=0.5}
  NamedRerouteDeclaration(None=SmoothStep(Value=SphereMask(A=@n22, B=@n23) {attenuationRadius=0.5, hardnessPercent=0})) {name=Shallow Wobble, nodeColor=<rgba 1 0.0234 0 1>}
  n24 = TextureSampleParameter2D() {
   desc="R = SpecMask G = Metallic B = Roughness"
   group=Gelatin
   parameterName=SpecularMasks
   samplerType=SAMPLERTYPE_Masks
   sortPriority=0
   texture=/Game/Characters/Player/Male/Medium/Bodies/M_MED_GelatinGummi/Textures/T_Gelatin_Blank_S
  }
  NamedRerouteDeclaration(None=@n24.R) {name=Specular, nodeColor=<rgba 1 0.0469 0 1>}
  NamedRerouteDeclaration(None=Constant() {r=1}) {name=Metallic, nodeColor=<rgba 1 0.1641 0 1>}
  NamedRerouteDeclaration(None=Multiply(A=@n24.B) {constB=0.3}) {name=Roughness, nodeColor=<rgba 0 0.9062 1 1>}
  n25 = ScalarParameter() {defaultValue=0.2, group=Gelatin, parameterName=Gelatin Deep Depth, sortPriority=30}
  n26 = ComponentMask(None=NamedRerouteUsage()) {r=true}
  n27 = ScalarParameter() {defaultValue=1, group=Gelatin, parameterName=Gelatin WobbleMult, sortPriority=8}
  n28 = Multiply(
   A=Multiply(A=@n25, B=Subtract(A=@n26) {constB=0.5})
   B=Multiply(A=ScalarParameter() {defaultValue=1, group=Gelatin, parameterName=Gelatin Deep Wobble Strength, sortPriority=30}, B=@n27)
  )
  n29 = ScalarParameter() {defaultValue=1, group=Gelatin, parameterName=Gelatin DepthMult, sortPriority=8}
  n30 = BumpOffset(Height=Constant(), HeightRatioInput=Multiply(A=Add(A=@n25, B=@n28), B=@n29)) {heightRatio=0.2}
  n31 = TextureSampleParameter2D(UVs=@n30) {
   group=Gelatin
   parameterName=FX
   samplerType=SAMPLERTYPE_Masks
   sortPriority=0
   texture=/Game/Characters/Player/Male/Medium/Bodies/M_MED_GelatinGummi/Textures/T_M_MED_GelatinGummi_Body_FX
  }
  n32 = TextureSampleParameter2D(UVs=@n30) {parameterName=FXNormals - tendrils, samplerType=SAMPLERTYPE_Normal, texture=&FlatNormalBlue}
  n33 = MaterialFunctionCall!(NormalIn=@n32.RGB) {materialFunction=&/MF_GelatinFakeLighting}
  n34 = Multiply(A=@n31.B, B=@n33.Result)
  n35 = ScalarParameter() {defaultValue=0.15, group=Gelatin, parameterName=Gelatin Mid Depth, sortPriority=20}
  n36 = NamedRerouteUsage()
  n37 = Add(A=ComponentMask(None=NamedRerouteUsage()) {r=true}, B=Multiply(A=ComponentMask(None=@n36) {r=true}) {constB=0.1})
  n38 = Multiply(
   A=Multiply(A=@n35, B=Subtract(A=@n37) {constB=0.5})
   B=Multiply(A=ScalarParameter() {defaultValue=0.666, group=Gelatin, parameterName=Gelatin Mid Wobble Strength, sortPriority=20}, B=@n27)
  )
  n39 = BumpOffset(Height=Constant(), HeightRatioInput=Multiply(A=Add(A=@n35, B=@n38), B=@n29)) {heightRatio=0.1}
  n40 = TextureSampleParameter2D(UVs=@n39) {
   group=Gelatin
   parameterName=FX
   samplerType=SAMPLERTYPE_Masks
   sortPriority=0
   texture=&T_M_MED_GelatinGummi_Body_FX
  }
  n41 = Multiply(A=@n40.G, B=ScalarParameter() {defaultValue=1, group=Gelatin, parameterName=Gelatin Mid Visilbility, sortPriority=20})
  n42 = TextureSampleParameter2D(UVs=@n39) {parameterName=FXNormals - tendrils, samplerType=SAMPLERTYPE_Normal, texture=&FlatNormalBlue}
  n43 = MaterialFunctionCall(NormalIn=@n42.RGB) {materialFunction=&MF_GelatinFakeLighting}
  n44 = ScalarParameter() {defaultValue=0.05, group=Gelatin, parameterName=Gelatin Shallow Depth, sortPriority=10}
  n45 = ComponentMask(None=NamedRerouteUsage()) {r=true}
  n46 = Multiply(
   A=Multiply(A=@n44, B=Subtract(A=@n45) {constB=0.5})
   B=Multiply(A=ScalarParameter() {defaultValue=1.5, group=Gelatin, parameterName=Gelatin Shallow Wobble Strength, sortPriority=10}, B=@n27)
  )
  n47 = BumpOffset(Height=Constant(), HeightRatioInput=Multiply(A=Add(A=@n44, B=@n46), B=@n29)) {heightRatio=0.1}
  n48 = TextureSampleParameter2D(UVs=@n47) {
   group=Gelatin
   parameterName=FX
   samplerType=SAMPLERTYPE_Masks
   sortPriority=0
   texture=&T_M_MED_GelatinGummi_Body_FX
  }
  n49 = Multiply(A=@n48.R, B=ScalarParameter() {defaultValue=1, group=Gelatin, parameterName=Gelatin Shallow Visibility, sortPriority=10})
  n50 = TextureSampleParameter2D(UVs=@n47) {parameterName=FXNormals - tendrils, samplerType=SAMPLERTYPE_Normal, texture=&FlatNormalBlue}
  n51 = MaterialFunctionCall(NormalIn=@n50.RGB) {materialFunction=&MF_GelatinFakeLighting}
  n52 = Add(A=Add(A=Multiply(A=@n34, B=@n33.highlight), B=Multiply(A=@n41, B=@n43.highlight)), B=Multiply(A=@n49, B=@n51.highlight))
  n53 = NamedRerouteUsage()
  n54 = OneMinus(None=Multiply(A=NamedRerouteUsage()) {constB=0.75})
  n55 = Multiply(
   A=Multiply(A=@n34, B=Multiply(A=@n53, B=@n54))
   B=ScalarParameter() {defaultValue=1, group=Gelatin, parameterName=Gelatin Deep Visilbility, sortPriority=30}
  )
  n56 = Multiply(A=NamedRerouteUsage()) {constB=0.75}
  n57 = Multiply(A=Multiply(A=@n40.G, B=@n43.Result), B=Multiply(A=NamedRerouteUsage(), B=OneMinus(None=@n56)))
  n58 = Multiply(A=@n48.R, B=@n51.Result)
  n59 = NamedRerouteUsage()
  n60 = Multiply(A=NamedRerouteUsage(), B=OneMinus(None=Multiply(A=@n59) {constB=0.75}))
  n61 = LinearInterpolate(
   A=LinearInterpolate(
    A=NamedRerouteDeclaration(None=@n55) {name=Deep, nodeColor=<rgba 0 0.0391 1 1>}
    B=NamedRerouteDeclaration(None=@n57) {name=Mid, nodeColor=<rgba 0.8516 0 1 1>}
    Alpha=@n41
   )
   B=NamedRerouteDeclaration(None=Multiply(A=@n58, B=@n60)) {name=Shallow, nodeColor=<rgba 0.3359 0 1 1>}
   Alpha=@n49
  )
  n62 = PixelNormalWS()
  n63 = Multiply(A=Multiply(A=Multiply(A=@n61, B=@n7)) {constB=1.5}, B=OneMinus(None=Fresnel(Normal=@n62) {baseReflectFraction=0, exponent=3}))
  n64 = VectorParameter() {group=Gelatin, parameterName=Gelatin Inner Fresnel Color, defaultValue=<rgba 0.005 0.0058 0.01 1>}
  n65 = VectorParameter() {group=Gelatin, parameterName=Gelatin Outer Fresnel Color, defaultValue=<rgba 0.4427 0.5773 0.8854 1>}
  n66 = PixelNormalWS()
  n67 = Fresnel(ExponentIn=ScalarParameter() {defaultValue=3, group=Gelatin, parameterName=Gelatin Frensel Exponent}, Normal=@n66) {baseReflectFraction=0, exponent=3}
  n68 = TextureSampleParameter2D() {group=Gelatin, parameterName=Emissive, sortPriority=5, texture=/Engine/EngineResources/Black_Low.Black}
  n69 = ScalarParameter() {defaultValue=1, group=Gelatin, parameterName=Gelatin Emissive Mult, sortPriority=5}
  n70 = ScalarParameter() {defaultValue=1, group=Gelatin, parameterName=Gelatin Emisive Frensel Exponent, sortPriority=5}
  n71 = ScalarParameter() {group=Gelatin, parameterName=Gelatin Emissive Base Fresnel, sortPriority=5}
  n72 = Add(
   A=Multiply(
    A=LinearInterpolate(A=@n64.RGB, B=@n65.RGB, Alpha=@n67)
    B=ScalarParameter() {defaultValue=1, group=Gelatin, parameterName=Gelatin Fresnel Mult}
   )
   B=Multiply(A=Multiply(A=@n68.RGB, B=@n69), B=Fresnel(ExponentIn=@n70, BaseReflectFractionIn=@n71, Normal=@n66) {baseReflectFraction=0, exponent=3})
  )
  n73 = ScalarParameter() {defaultValue=0.5, group=Gelatin, parameterName=Gelatin Emissive Wobble Mult, sortPriority=5}
  n74 = Add(A=Multiply(A=@n63) {constB=0.333}, B=LinearInterpolate(A=Multiply(A=@n72, B=@n73) {constB=0.5}, B=@n72, Alpha=NamedRerouteUsage()))
  NamedRerouteDeclaration(None=Desaturation(None=Add(A=@n52, B=@n74), Fraction=Constant() {r=0.2})) {name=Emissive, nodeColor=<rgba 0.1484 0 1 1>}
  NamedRerouteDeclaration(None=@n24.G) {name=FX Holdout, nodeColor=<rgba 0 0.3672 1 1>}
  n75 = Multiply(A=@n16, B=ScalarParameter() {defaultValue=0.5, group=Gelatin, parameterName=Gelatin Shallow ColorPulse Speed, sortPriority=10})
  n76 = CurveAtlasRowParameter(CurveTime=Frac(None=Add(A=@n15, B=@n75))) {
   atlas=&AthenaCurveAtlas01
   curve=/Game/Characters/Player/Male/Medium/Bodies/M_MED_GelatinGummi/Textures/T_GelatinGummi_Shallow
   defaultValue=323
   group=Gelatin
   parameterName=Gelatin Shallow ColorCurve
   sortPriority=10
  }
  NamedRerouteDeclaration(None=@n76) {name=ColorPulse Shallow, nodeColor=<rgba 0 0.9062 1 1>}
  n77 = Multiply(A=@n16, B=ScalarParameter() {defaultValue=0.5, group=Gelatin, parameterName=Gelatin Deep ColorPulse Speed, sortPriority=30})
  n78 = CurveAtlasRowParameter(CurveTime=Frac(None=Add(A=@n15, B=@n77))) {
   atlas=&AthenaCurveAtlas01
   curve=&T_GelatinGummi_Shallow
   defaultValue=323
   group=Gelatin
   parameterName=Gelatin Deep ColorCurve
   sortPriority=30
  }
  NamedRerouteDeclaration(None=@n78) {name=ColorPulse Deep, nodeColor=<rgba 1 0.5391 0 1>}
  n79 = Frac(None=Add(A=@n21, B=ScalarParameter() {group=Gelatin, parameterName=Gelatin Mid Wobble Offset, sortPriority=20}))
  n80 = Constant() {r=0.5}
  NamedRerouteDeclaration(None=SmoothStep(Value=SphereMask(A=@n79, B=@n80) {attenuationRadius=0.5, hardnessPercent=0})) {name=Mid Wobble, nodeColor=<rgba 0.7109 0 1 1>}
  n81 = Frac(None=Add(A=@n21, B=ScalarParameter() {defaultValue=-0.1, group=Gelatin, parameterName=Gelatin Deep Wobble Offset, sortPriority=30}))
  n82 = Constant() {r=0.5}
  NamedRerouteDeclaration(None=SmoothStep(Value=SphereMask(A=@n81, B=@n82) {attenuationRadius=0.5, hardnessPercent=0})) {name=Deep Wobble, nodeColor=<rgba 0.4297 1 0 1>}
  n83 = TextureSampleParameter2D() {
   group=Gelatin
   parameterName=Normal
   samplerType=SAMPLERTYPE_Normal
   sortPriority=0
   texture=&FlatNormalBlue
  }
  n84 = SetMaterialAttributes!(
   MaterialAttributes=FunctionInput() {inputType=FunctionInput_MaterialAttributes}
   BaseColor=@n63
   Metallic=NamedRerouteUsage()
   Specular=NamedRerouteUsage()
   Roughness=NamedRerouteUsage()
   EmissiveColor=NamedRerouteUsage()
   Normal=@n83.RGB
   Opacity
  ) {attributeSetTypes [
    #1
    #2
    #3
    #4
    #5
    #6
    #7
   ]}
  NamedRerouteDeclaration(None=@n84) {name=Gelatin Material, nodeColor=<rgba 0 1 0.9297 1>}
  n85 = NamedRerouteUsage()
  Saturate(None=MaterialFunctionCall(Input=Multiply() {constB=3.5}, Input Low=OneMinus(None=@n85)) {materialFunction=&RemapValueRange}.Result)
  TextureSampleParameter2D() {group=Gelatin, parameterName=FX, sortPriority=0, texture=/Game/Athena/Environments/Nexus/Textures/T_Tile_Dots}
  TextureSampleParameter2D() {
   group=Gelatin
   parameterName=FX
   samplerType=SAMPLERTYPE_Masks
   sortPriority=0
   texture=&T_M_MED_GelatinGummi_Body_FX
  }
 ]}
