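# M_Gelatin is a Default Lit character material for the medium male GelatinGummi player body, flagged for skeletal mesh, morph target and mesh deformer use.
# It uses material attributes, passing the shared character PreFX function's result through MF_Gelatin and into the shared PostFX function.
# The PreFX AO output drives all of PostFX's AO, Distance Blend, Fresnel and Gmap inputs.
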Material /Game/Characters/Player/Male/Medium/Bodies/M_MED_GelatinGummi/Materials/M_Gelatin
graph {
 properties {
  bUseMaterialAttributes=true
  bUsedWithMeshDeformer=true
  bUsedWithMorphTargets=true
  bUsedWithSkeletalMesh=true
  pixelDepthOffsetMode=PDOM_Legacy
  usedShadingModels=Default Lit
 }
 nodes [
  n1 = MaterialFunctionCall() {materialFunction=&BaseFN_CharacterFunction_PreFX}
  MaterialFunctionCall(
   AO=@n1.AO
   Distance Blend=@n1.AO
   Fresnel=@n1.AO
   Gmap=@n1.AO
   Material=MaterialFunctionCall!(In=@n1.Result) {materialFunction=&/MF_Gelatin}.Result
  ) {materialFunction=&BaseFN_CharacterFunction_PostFX}
 ]
}
